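# Master material for the Scrawl backpack cosmetic: a masked, two-sided character material that animates an outer and an interior layer drawing themselves in.
# Looping time, quantized by an FPS parameter, thresholds each layer's Scrawl color texture against staggered timing, multiplied by opacity textures and combined as the mask.
# Panned cloud noise builds stepped wiggle UVs, while derived inner and drying masks feed named reroutes.
# Layer-blended PBR inputs pass through shared character FX functions.
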Material /BRCosmetics/Accessories/FORT_Backpacks/Backpack_M_MED_Scrawl/Material/M_MED_Scrawl_Master
graph {
 properties {
  bUseMaterialAttributes=true
  bUsedWithMeshDeformer=true
  bUsedWithSkeletalMesh=true
  blendMode=BLEND_Masked
  opacityMaskClipValue=0.3
  pixelDepthOffsetMode=PDOM_Legacy
  refractionMethod=RM_IndexOfRefraction
  twoSided=true
  usedShadingModels=Default Lit
 }
 nodes [
  n1 = NamedRerouteUsage()
  LinearInterpolate(Alpha=@n1) {constA=4, constB=3}
  n2 = MaterialFunctionCall() {materialFunction=&BaseFN_CharacterFunction_PreFX}
  n3 = TextureSampleParameter2D(UVs=NamedRerouteUsage()) {
   group=Diffuse
   parameterName=Diffuse
   texture=/BRCosmetics/Accessories/FORT_Backpacks/Backpack_M_MED_Scrawl/Textures/T_M_MED_Scrawl_Backpack_D
  }
  n4 = Desaturation(None=Multiply(A=@n3.RGB) {constB=4}, Fraction=LinearInterpolate(Alpha=@n1) {constA=-0.2, constB=0})
  n5 = TextureSampleParameter2D(UVs=NamedRerouteUsage()) {
   group=Diffuse
   parameterName=Diffuse 2
   texture=/BRCosmetics/Accessories/FORT_Backpacks/Backpack_M_MED_Scrawl/Textures/Interior/T_M_MED_Scrawl_Backpack_Interior_D
  }
  n6 = LinearInterpolate(A=@n5.RGB, B=Multiply(A=@n5.RGB) {constB=1.5}, Alpha=NamedRerouteUsage())
  n7 = NamedRerouteUsage()
  n8 = TextureSampleParameter2D(UVs=NamedRerouteUsage()) {
   group=Specular
   parameterName=SpecularMasks
   samplerType=SAMPLERTYPE_Masks
   texture=/BRCosmetics/Accessories/FORT_Backpacks/Backpack_M_MED_Scrawl/Textures/T_M_MED_Scrawl_Backpack_S
  }
  n9 = TextureSampleParameter2D(UVs=NamedRerouteUsage()) {
   group=Specular
   parameterName=SpecularMasks 2
   samplerType=SAMPLERTYPE_Masks
   texture=/BRCosmetics/Accessories/FORT_Backpacks/Backpack_M_MED_Scrawl/Textures/Interior/T_M_MED_Scrawl_Backpack_Interior_S
  }
  n10 = MaterialFunctionCall(Color=LinearInterpolate(A=@n8.RGB, B=@n9.RGB, Alpha=NamedRerouteUsage())) {materialFunction=&SplitComponents}
  n11 = TextureSampleParameter2D(UVs=NamedRerouteUsage()) {
   group=Normal
   parameterName=Normals
   samplerType=SAMPLERTYPE_Normal
   texture=/BRCosmetics/Accessories/FORT_Backpacks/Backpack_M_MED_Scrawl/Textures/T_M_MED_Scrawl_Backpack_N
  }
  n12 = TextureSampleParameter2D(UVs=NamedRerouteUsage()) {
   group=Normal
   parameterName=Normals 2
   samplerType=SAMPLERTYPE_Normal
   texture=/BRCosmetics/Accessories/FORT_Backpacks/Backpack_M_MED_Scrawl/Textures/Interior/T_M_MED_Scrawl_Backpack_Interior_N
  }
  n13 = OneMinus(None=NamedRerouteUsage())
  n14 = TextureSample(UVs=NamedRerouteUsage()) {samplerType=SAMPLERTYPE_LinearColor, texture=/BRCosmetics/Accessories/FORT_Backpacks/Backpack_M_MED_Scrawl/Textures/Scrawl_color_Outside}
  n15 = Constant()
  n16 = Time() {bOverride_Period=true, period=10}
  n17 = ScalarParameter() {defaultValue=12, parameterName=FPS}
  n18 = Divide(A=Ceil(None=Multiply(A=@n16, B=@n17)), B=@n17)
  n19 = Divide(A=@n18) {constB=3}
  n20 = Multiply(
   A=TextureSample(UVs=NamedRerouteUsage()) {texture=/BRCosmetics/Accessories/FORT_Backpacks/Backpack_M_MED_Scrawl/Textures/T_M_MED_Scrawl_Backpack_Opacity}.R
   B=If(A=@n14.R, B=@n19, "A > B"=@n15, "A == B"=@n15, "A < B"=Constant() {r=1})
  )
  n21 = TextureCoordinate()
  n22 = Time()
  n23 = ScalarParameter() {defaultValue=12, parameterName=FPS}
  n24 = Divide(A=Ceil(None=Multiply(A=@n22, B=@n23)), B=@n23)
  n25 = TextureSample(UVs=Panner(Time=@n24) {speedX=2.314, speedY=0.316}) {texture=/Game/Global/Textures/Default/Noise/T_Noise_Cloud}
  n26 = Multiply(A=ConstantBiasScale(None=ComponentMask(None=@n25.RGB) {g=true, r=true}) {bias=-0.5, scale=2}) {constB=0.003}
  n27 = TextureSample(UVs=NamedRerouteDeclaration(None=Add(A=@n21, B=@n26)) {name=Layer2 Wiggle UV, nodeColor=<rgba 0.3286 0.2415 0.8281 1>}) {
   samplerType=SAMPLERTYPE_LinearColor
   texture=/BRCosmetics/Accessories/FORT_Backpacks/Backpack_M_MED_Scrawl/Textures/Interior/Scrawl_color_Inside
  }
  n28 = Divide(A=Subtract(A=@n18) {constB=3}) {constB=5}
  n29 = Constant()
  n30 = Multiply(
   A=If(A=@n27.R, B=@n28, "A > B"=@n29, "A == B"=@n29, "A < B"=Constant() {r=1})
   B=TextureSample(UVs=NamedRerouteUsage()) {texture=/BRCosmetics/Accessories/FORT_Backpacks/Backpack_M_MED_Scrawl/Textures/Interior/T_M_MED_Scrawl_Backpack_Interior_Opacity}.R
  )
  n31 = SetMaterialAttributes!(
   MaterialAttributes=@n2.Result
   Base Color=Multiply(A=LinearInterpolate(A=@n4, B=@n6, Alpha=@n7))
   Metallic=LinearInterpolate(Alpha=@n7) {constA=1, constB=0}
   Specular=@n10.R
   Roughness=LinearInterpolate(A=Multiply(A=@n10.B) {constB=0.8}, Alpha=NamedRerouteUsage()) {constB=0.5}
   Normal=MaterialFunctionCall(
    BaseNormal=@n11.RGB
    AdditionalNormal=MaterialFunctionCall(Normal=@n12.RGB, Flatness=@n13) {materialFunction=&FlattenNormal}.Result
   ) {materialFunction=&BlendAngleCorrectedNormals}.Result
   Opacity Mask=Max(A=@n20, B=@n30)
  ) {attributeSetTypes [
    #1
    #2
    #3
    #4
    #5
    #6
   ]}
  n32 = MaterialFunctionCall(
   AO=@n2.AO
   Distance Blend=@n2.AO
   Fresnel=@n2.AO
   Gmap=@n2.AO
   Pre Skinned Local Position=@n2.AO
   Material=MaterialFunctionCall(
    PreskinnedPosition=@n2.Pre skinned Local Position
    Gmap=@n2.Pre skinned Local Position
    MaterialAttributes=@n31
    AO=@n2.Pre skinned Local Position
   ) {materialFunction=&CharacterSkinFX}.Result
  ) {materialFunction=&BaseFN_CharacterFunction_PostFX}
  MaterialFunctionCall(In=@n32.Result) {materialFunction=&CINE_MF_CharPostFX}
  n33 = Subtract(A=@n30, B=@n20)
  n34 = TwoSidedSign()
  NamedRerouteDeclaration(None=LinearInterpolate(A=Saturate(None=@n33), B=@n30, Alpha=Saturate(None=@n34))) {name=Inner, nodeColor=<rgba 0.5703 1 0 1>}
  n35 = Panner(Time=@n24) {speedY=3.14}
  n36 = ConstantBiasScale(None=ComponentMask(None=TextureSample(UVs=@n35) {texture=&T_Noise_Cloud}.RGB) {g=true, r=true}) {bias=-0.5, scale=2}
  NamedRerouteDeclaration(None=Add(A=@n21, B=Multiply(A=@n36) {constB=0.003})) {name=Layer1 Wiggle UV, nodeColor=<rgba 0.1484 0 1 1>}
  n37 = ScalarParameter() {defaultValue=0.3, parameterName=Drying Time}
  n38 = Subtract(A=@n19, B=@n37)
  n39 = Subtract(A=@n28, B=@n37)
  NamedRerouteDeclaration(None=LinearInterpolate(A=SmoothStep(Min=@n38, Max=@n19, Value=@n14.R), B=SmoothStep(Min=@n39, Max=@n28, Value=@n27.R), Alpha=NamedRerouteUsage())) {name=Drying Anim, nodeColor=<rgba 0.2891 0 1 1>}
  ScalarParameter() {defaultValue=1.2}
 ]
}
